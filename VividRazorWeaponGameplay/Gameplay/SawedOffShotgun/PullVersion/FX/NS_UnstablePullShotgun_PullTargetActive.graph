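# Niagara effect for an active pull target of the Unstable Pull Shotgun, drawing purple sprites from the owner toward a target point.
# Attractor spawns sprites around the owner on a decaying looped rate, then pulls them to System.TargetPosition with attraction, early curl noise and drag, killing them on arrival.
# Sprites stretch with speed, follow source movement and fade as owner velocity rises.
# System modules include skeletal mesh sampling; the Streaks emitter is disabled.
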NiagaraSystem /VividRazorWeaponGameplay/Gameplay/SawedOffShotgun/PullVersion/FX/NS_UnstablePullShotgun_PullTargetActive
dependencies {
 usedDataInterfaces [&NiagaraDataInterfaceCurve]
 usedDynamicInputs [&DistanceBetweenPositions]
 usedModules [
  &SampleSkeletalMesh
  &self:SystemSpawnScript.SystemScriptSource.SystemScriptGraph.NiagaraNodeAssignment_0.SetVariables_#1
  &self:SystemSpawnScript.SystemScriptSource.SystemScriptGraph.NiagaraNodeAssignment_2.SetVariables_#2
  &self:SystemSpawnScript.SystemScriptSource.SystemScriptGraph.NiagaraNodeAssignment_3.SetVariables_#3
  &self:SystemSpawnScript.SystemScriptSource.SystemScriptGraph.NiagaraNodeAssignment_1.SetVariables_#4
  &SystemState
 ]
}
emitters {
 Attractor {
  emitterUpdateScript [
   &EmitterState(Life Cycle Mode=::NewEnumerator1, Loop Behavior=::NewEnumerator1, Loop Delay=1, Loop Duration=0.5, UseLoopDelay=false)
   &SpawnBurst_Instantaneous(Spawn Count=10, Spawn Probability=$System.Timeline001.InverseNormalizedTime, Use Spawn Probability=true, UseLoopCountLimit=false) {enabled=false}
   &SpawnRate(
    SpawnRate=&Lerp_Float(
     A=0
     B=25
     Alpha=&FloatFromCurve(
      FloatCurve=DataInterface() {propertyValues {Curve {keys [
          {interpMode=RCIM_Cubic, tangentMode=RCTM_User, time=0.1, value=1.125}
          {arriveTangent=-1.0633, interpMode=RCIM_Cubic, leaveTangent=-1.0633, tangentMode=RCTM_User, time=1.5, value=0}
         ]}}}
      CurveIndex=$Emitter.NormalizedLoopAge
     )
    )
    Use Spawn Probability=false
   )
  ]
  particleSpawnScript [
   &InitializeParticle(
    Color=<rgba 1 0.55 0.7 1>
    Lifetime=0.3
    Sprite Size Max=<2 8>
    Sprite Size Min=<2 5>
    Sprite Size Mode=::NewEnumerator2
    UsePositionOffset=false
    Write Material Random=true
   )
   &SystemLocation(MaskSpawnGroup=false) {enabled=false}
   &ShapeLocation(
    MaskSpawnGroup=false
    Offset Coordinate Space=::Local
    Offset Mode=::NewEnumerator0
    Shape Origin=$Engine.Owner.Position
    Sphere Distribution=::NewEnumerator1
    Sphere Radius=50
    U Position=&RandomRangeFloat(Minimum=-0.25, Maximum=0.25)
    UseOwnerScale=true
    V Position=&RandomRangeFloat(Minimum=0, Maximum=1)
   )
  ]
  particleUpdateScript [
   &ParticleState()
   &ScaleSpriteSize(
    Scale Sprite Size Mode=::NewEnumerator0
    Uniform Scale Factor=&Lerp_Float(
     A=1
     B=0
     Alpha=&FloatFromCurve(
      FloatCurve=DataInterface() {propertyValues {Curve {keys [{interpMode=RCIM_Cubic, tangentMode=RCTM_User, time=0.3, value=0}, <at 0.6 1>]}}}
      CurveIndex=&NormalizeFloat(
       Value=&Abs_Float(Float:NiagaraFloat=&VectorLength(Vector=$Engine.Owner.Velocity))
       Min=0
       Max=1000
      )
     )
    )
   ) {enabled=false, module=ScaleSpriteSize001}
   &ScaleSpriteSize(Scale Sprite Size Mode=::NewEnumerator0, Uniform Scale Factor=1.5)
   &PointAttractionForce(
    Attraction Radius=2500
    Attraction Strength=&Lerp_Float(
     A=50
     B=700
     Alpha=&FloatFromCurve(FloatCurve=DataInterface() {propertyValues {Curve {keys [{interpMode=RCIM_Cubic, tangentMode=RCTM_User, time=0.3, value=0}, {arriveTangent=2.5682, tangentMode=RCTM_User, time=0.86, value=1}]}}})
    )
    Attractor Position:NiagaraPosition=$System.TargetPosition
    Falloff Exponent=0.6
    Kill Radius=6
    Kill Within Radius=true
    Module.Use Falloff=true
    UseAttractorOffset=false
   )
   &CurlNoiseForce(
    Bias Noise=false
    Noise Frequency=40
    Noise Strength=&Multiply_Float(
     A=2000
     B=&FloatFromCurve(FloatCurve=DataInterface() {propertyValues {Curve {keys [
          {arriveTangent=8.6523, interpMode=RCIM_Cubic, leaveTangent=8.6523, tangentMode=RCTM_User, time=0, value=0.3}
          {arriveTangent=-1, interpMode=RCIM_Cubic, leaveTangent=-1, time=0.1, value=1}
          {interpMode=RCIM_Cubic, tangentMode=RCTM_User, time=0.3, value=0}
         ]}}})
    )
    Pan Noise=false
   )
   &Drag(
    Drag=&FloatFromCurve(FloatCurve=DataInterface() {propertyValues {Curve {keys [<at 0 1>,<at 1 0>]}}}, Scale Curve=5)
    Use Linear Drag=true
    Use Rotational Drag=true
   )
   &InheritVelocity(
    Inherited Velocity Amount Scale=&VectorFromFloat(Value=1)
    Inherited Velocity Speed Limit=20
    Speed Limit=true
    UseSpeedThreshold=false
   ) {enabled=false}
   &InheritSourceMovement(
    Applied Position Scale=&VectorFromFloat(Value=0.6)
    ApplyForce=false
    ApplyPosition=true
    ApplyPositionFalloff=false
   )
   &SolveForcesAndVelocity(Clamp Velocity=false, Limit Acceleration=false)
   &ScaleSpriteSizeBySpeed(Max Scale Factor=<1 2.5>, Min Scale Factor=<0.4 0.4>, Velocity Threshold=500)
   &FN_Sprite_Attributes(
    Color=<rgba 1.1956 0.4447 2 0.1>
    Create Color Attribute=true
    Create SpriteSize Attribute=true
    SpriteSize=&Multiply_Vector2D(A=$Particles.SpriteSize, B=<20 3>)
    Write to Module Namespace=true
   )
   &DynamicMaterialParameters(
    Index 0 Param 1=2
    Module.Index0Param0Write=true
    Module.Index0Param1Write=true
    Module.Index0Param2Write=true
    Module.Index0Param3Write=true
   )
   &ScaleColor(
    Scale Alpha=&FloatFromCurve(
     FloatCurve=DataInterface() {propertyValues {Curve {keys [{interpMode=RCIM_Cubic, tangentMode=RCTM_User, time=0.33, value=1}, <at 0.45 0>]}}}
     CurveIndex=&NormalizeFloat(
      Value=&Abs_Float(Float:NiagaraFloat=&VectorLength(Vector=$Engine.Owner.Velocity))
      Min=0
      Max=1000
     )
    )
    ScaleA=true
    ScaleRGB=false
   )
   /SetVariables(Particles.ColorGlow:LinearColor=&MultiplyLinearColor(
     Linear Color=$Particles.FN_Sprite_Attributes.Color
     Scale Factor=&MakeVector4FromVectorAndFloat(
      "Vector (RGB)"=<1 1 1>
      "Float (Alpha)"=&FloatFromCurve(
       FloatCurve=DataInterface() {propertyValues {Curve {keys [{interpMode=RCIM_Cubic, tangentMode=RCTM_User, time=0.33, value=1}, <at 0.45 0>]}}}
       CurveIndex=&NormalizeFloat(
        Value=&Abs_Float(Float:NiagaraFloat=&VectorLength(Vector=$Engine.Owner.Velocity))
        Min=0
        Max=1000
       )
       Scale Curve=0.5
      )
     )
    )) {isSetParameters=true}
  ]
  rendererClasses [&NiagaraSpriteRendererProperties]
  renderers [
   {rendererClass=&NiagaraSpriteRendererProperties}
   {rendererClass=&NiagaraSpriteRendererProperties, rendererIndex=1}
  ]
 }
 Streaks {
  bEnabled=false
  emitterUpdateScript [
   &EmitterState(Life Cycle Mode=::NewEnumerator1, Loop Behavior=::NewEnumerator1, Loop Duration=1, UseLoopDelay=false)
   &SpawnRate(
    SpawnRate=&FloatFromCurve(
     FloatCurve=DataInterface() {propertyValues {Curve {keys [
         {interpMode=RCIM_Cubic, tangentMode=RCTM_Break, time=0, value=0}
         {interpMode=RCIM_Cubic, tangentMode=RCTM_Break, time=0.4, value=1}
         {arriveTangent=-3.9775, interpMode=RCIM_Cubic, tangentMode=RCTM_Break, time=1, value=0}
        ]}}}
     CurveIndex=$Emitter.NormalizedLoopAge
     Scale Curve=50
    )
    Use Spawn Probability=false
   )
  ]
  particleSpawnScript [
   /SetVariables(Particles.PosAttractOffset:Vector3f=&RandomRangeVector(Minimum=<0 0 -100>, Maximum=<0 0 -100>)) {isSetParameters=true}
   &InitializeParticle(
    Color Maximum=<rgba 0.9271 0.9694 1 1>
    Color Minimum=<rgba 0.0894 0.1757 1 1>
    Color Mode=::NewEnumerator2
    Lifetime Max=0.25
    Lifetime Min=0.1
    Lifetime Mode=::NewEnumerator1
    Sprite Size Max=<9 9>
    Sprite Size Min=<6 6>
    Sprite Size Mode=::NewEnumerator2
    UsePositionOffset=false
    Write Material Random=true
   )
   &ShapeLocation(MaskSpawnGroup=false, Override Seed=false, UseOwnerScale=true)
   &KillParticles(Kill Particles=&SetBoolByFloatComparison(
     A=&DotProduct(
      Vector A=&RotateVectorByQuaternion(Vector To Rotate=<1 0 0>, Quaternion=$System.SampleSkeletalMesh.SampledRotation)
      Vector B=&NormalizeVector(
       Vector To Normalize=&Subtract_Vector(
        A=Linked() {linkedVariable {name=Engine.Owner.Position, type=NiagaraPosition}}
        B=Linked() {linkedVariable {name=Particles.Position, type=NiagaraPosition}}
       )
       Safe Normalize=true
       Safe Fallback Threshold=0
       Safe Fallback Vector=<0 0 0.001>
      )
     )
     B=0
     Comparison Type=::NewEnumerator6
    )) {enabled=false}
   &AddVelocity(Velocity=$Engine.Owner.SystemXAxis, Velocity Speed Scale=-100)
  ]
  particleUpdateScript [
   &ParticleState()
   &KillParticlesInVolume(
    Invert Volume=true
    Kill Shape=::NewEnumerator2
    Plane Normal:Vector3f=<0 1 0>
    Volume Origin:NiagaraPosition=$System.SampleSkeletalMesh.SampledPosition
   ) {enabled=false}
   &PointAttractionForce(
    Attraction Radius=2000
    Attraction Strength=&RandomRangeFloat(Minimum=300, Maximum=400)
    Attractor Position:NiagaraPosition=$System.TargetPosition
    Kill Radius=100
    Kill Within Radius=true
    Module.Use Falloff=true
    UseAttractorOffset=false
   ) {enabled=false}
   &SolveForcesAndVelocity(Clamp Velocity=false, Limit Acceleration=false)
   &ScaleSpriteSize(
    Scale Factor=&MakeVector2D(
     X=2
     Y=&Lerp_Float(
      A=15
      B=50
      Alpha=&FloatFromCurve(
       FloatCurve=DataInterface() {propertyValues {Curve {keys [
           {interpMode=RCIM_Cubic, tangentMode=RCTM_Break, time=0.1, value=0}
           {arriveTangent=0.7475, interpMode=RCIM_Cubic, leaveTangent=0.7475, tangentMode=RCTM_User, time=0.8, value=0.2}
           {arriveTangent=4.5178, interpMode=RCIM_Cubic, tangentMode=RCTM_Break, time=1, value=1}
          ]}}}
       CurveIndex=&RandomRangeFloat(Minimum=0, Maximum=1)
       Scale Curve=1.5
      )
     )
    )
    Scale Sprite Size Mode=::NewEnumerator2
   )
   &ScaleColor(
    Scale Alpha=&FloatFromCurve(
     FloatCurve=DataInterface() {propertyValues {Curve {keys [
         {interpMode=RCIM_Cubic, tangentMode=RCTM_Break, time=0, value=0}
         {interpMode=RCIM_Cubic, tangentMode=RCTM_Break, time=0.4, value=1}
         {arriveTangent=-4.3816, interpMode=RCIM_Cubic, tangentMode=RCTM_Break, time=1, value=0}
        ]}}}
     Scale Curve=0.4
    )
    Scale RGB=&VectorFromFloat(Value=&Lerp_Float(
      A=0.5
      B=1.5
      Alpha=&FloatFromCurve(
       FloatCurve=DataInterface() {propertyValues {Curve {keys [
           {interpMode=RCIM_Cubic, tangentMode=RCTM_Break, time=0, value=0}
           {arriveTangent=0.9115, interpMode=RCIM_Cubic, leaveTangent=0.9115, tangentMode=RCTM_User, time=0.8, value=0.2}
           {arriveTangent=4.1785, interpMode=RCIM_Cubic, tangentMode=RCTM_Break, time=1, value=1}
          ]}}}
       CurveIndex=&RandomRangeFloat(Minimum=0, Maximum=1)
      )
     ))
    ScaleA=true
    ScaleRGB=true
   )
  ]
  rendererClasses [&NiagaraSpriteRendererProperties]
  renderers [{rendererClass=&NiagaraSpriteRendererProperties}]
 }
}
summary {
 systemName=NS_UnstablePullShotgun_PullTargetActive
 emitters [
  {emitterName=Attractor, rendererClasses [&NiagaraSpriteRendererProperties]}
  {bEnabled=false, emitterName=Streaks, rendererClasses [&NiagaraSpriteRendererProperties]}
 ]
 userVariables [
  User.PhysicalSurface {type=NiagaraInt32}
  User.TargetWeapon {
   type=NiagaraDataInterfaceSkeletalMesh
   defaultValue {
    struct=NiagaraExt_VariableValue_DataInterface
    value {dataInterface=&self:NiagaraDataInterfaceSkeletalMesh_0}
   }
  }
 ]
}
