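; Event-triggered gameplay ability for the pull variant of a sawed-off shotgun that yanks hit hostile pawns toward the wielder.
; Fire events update a bullet count; damage events from the source weapon are filtered to hostile pawns and gated by a trigger check.
; Targets gain fall-damage immunity and are launched along a Z-offset vector toward the avatar, with tag-guarded burst cues on both pawns.
; An interpolation-based motion function exists but is not called here.
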(event Ability|EventActivateAbilityFromEvent
  (params (EventData :type Gameplay%20Event%20Data%20Structure%20%28by%20ref%29))
  (bind (_currentfortpawntarget _gc_pulledtargetcue _returnvalue) (Utilities|GetTargetCueParameters))
  (bind _event_data (Variables|Default|GetEvent%20Data))
  (bind _currentfortpawntarget_2 (Variables|Default|GetCurrentFortPawnTarget))
  (bind _output_get (Variables|Default|SetEvent%20Data EventData))
  (if (== (Utilities|Struct|BreakGameplayEventData _output_get) "(TagName=\"Event.Weapons.Fire\")")
    (CallFunction|HandleBulletCount))
  (if (Default|IsDamgefromSourceWeapon _event_data)
    (bind _outhitactors (Utilities|HasValidGroupOfHitActors (CallFunction|GetHitActorsfromEventData _event_data))
      (:true
        (if (Utilities|ShouldEffectTrigger?)
          (CallFunction|PlayGameplayCueonSelf)
          (for _array_element _outhitactors
            (bind _fortpawn (Class|FortQuestBlueprintLibrary|GetFortPawnFromObject _array_element)
              (:Valid
                (bind _output_get_2 (Variables|Default|SetCurrentFortPawnTarget _fortpawn))
                (CallFunction|ApplyFallDamageImmunity _output_get_2)
                (CallFunction|ApplyMotionThroughLaunch _currentfortpawntarget_2)
                (if (not (GameplayTags|HasMatchingGameplayTag (Ability|GetAbilitySystemComponent _currentfortpawntarget_2) (Variables|Default|GetTag%20to%20Check)))
                  ; Applying GE with effect triggered tag
                  (GameplayEffects|ApplyGameplayEffectToTarget :self (Ability|GetAbilitySystemComponent (Ability|GetAvatarActorfromActorInfo)) :GameplayEffectClass (Variables|Default|GetGE_ApplyGPCue) :Target (Ability|GetAbilitySystemComponent (Variables|Default|GetCurrentFortPawnTarget)) :Level 1.0)
                  (GameplayCue|ExecuteGameplayCueOnActor(Burst) _currentfortpawntarget _gc_pulledtargetcue _returnvalue))
                (Ability|EndAbility))
              (:NotValid
                (Ability|EndAbility :!label _label))))))
      (:false
        _)))
  (goto _label))

(fn ApplyFallDamageImmunity
  (params (HitActor :type Actor%20Object%20Reference))
  (bind _returnvalue (Spec|MakeSpecHandle(ByClass) (Variables|Default|GetGE_FallDamageImmunity) (Ability|GetAvatarActorfromActorInfo)))
  (GameplayEffects|ApplyGameplayEffectSpecToTarget (Ability|GetFortAbilitySystemComponent HitActor) _returnvalue (Ability|GetFortAbilitySystemComponent HitActor)))

(fn ApplyMotionThroughInterpolation
  (params (HitTarget :type Fort%20Pawn%20Object%20Reference))
  (bind (_returnvalue_x _returnvalue_y) (Math|Vector|BreakVector (* (Transformation|GetActorForwardVector (Ability|GetAvatarActorfromActorInfo)) (ScalableFloat|GetValueatLevel (Variables|Default|GetHFTargetLocYOffset)))))
  (Transformation|SetActorLocation HitTarget (Math|Interpolation|VInterpTo (Transformation|GetActorLocation HitTarget) (Utilities|Operators|Add (Transformation|GetActorLocation (Ability|GetAvatarActorfromActorInfo)) (Math|Vector|MakeVector 0.0 _returnvalue_y (ScalableFloat|GetValueatLevel (Variables|Default|GetHFTargetLocZOffset)))) (Utilities|Time|GetWorldDeltaSeconds) (ScalableFloat|GetValueatLevel (Variables|Default|GetHFTargetLocInterpolationSpeed))) false false))

(fn ApplyMotionThroughLaunch
  (params (HitTarget :type Fort%20Pawn%20Object%20Reference))
  (bind (_launchintensityzoffset _launchintensity) (Utilities|SelectLaunchIntensities (Variables|Default|GetCurrentFortPawnTarget)))
  (bind _output_get (LocalVariables|SetLaunchHitTarget HitTarget))
  (Utilities|LaunchPawns 0 (LocalVariables|GetLaunchHitTarget) (* (Math|Vector|Normalize (Utilities|Operators|Add (- (Transformation|GetActorLocation (Ability|GetAvatarActorfromActorInfo)) (Transformation|GetActorLocation _output_get)) (Math|Vector|MakeVector 0.0 0.0 _launchintensityzoffset))) _launchintensity) true true true true false true))

(fn CalculateNormalizedVectorBetweenPawns
  (params (TargetActor :type Actor%20Object%20Reference))
  (return (Math|Vector|Normalize (- (Transformation|GetActorLocation (Ability|GetAvatarActorfromActorInfo)) (Transformation|GetActorLocation TargetActor)))))

(fn GetAvatarAsFortPawn
  (bind _asfort_pawn (Utilities|Casting|CastToFortPawn (Ability|GetAvatarActorfromActorInfo))
    (:then
      (return _asfort_pawn))
    (:CastFailed
      (return 0))))

(fn GetAvatarController
  (bind _outfortplayerpawn (CallFunction|GetAvatarasValidFortPlayerPawn)
    (:ValidFortPlayerPawn
      (Utilities|IsValid (Fortnite|GetFortPlayerControllerfromActor _outfortplayerpawn)
        (:Is%20Valid
          (return (Fortnite|GetFortPlayerControllerfromActor _outfortplayerpawn)))
        (:Is%20Not%20Valid
          (Utilities|IsValid (Pawn|GetController _outfortplayerpawn)
            (:Is%20Valid
              (return (Pawn|GetController _outfortplayerpawn)))
            (:Is%20Not%20Valid
              _)))))
    (:AvatarCastFailed
      _))
  (return 0))

(fn GetHitActorsFromEventData
  (params (GameplayEventData :type Gameplay%20Event%20Data%20Structure%20%28by%20ref%29))
  (bind (_eventtag _instigator _target) (Utilities|Struct|BreakGameplayEventData GameplayEventData))
  (bind _output_get (LocalVariables|SetLocalTarget _target))
  (Utilities|IsValid _output_get
    (:Is%20Valid
      (bind _returnvalue (Default|GetValidAvatarController true)
        (:Is%20Valid
          (if (CallFunction|IsPawnandNotFriendly _returnvalue (LocalVariables|GetLocalTarget))
            (return (Utilities|Array|MakeArray (LocalVariables|GetLocalTarget)))
            (else
              (return 0))))))
    (:Is%20Not%20Valid
      (return 0))))

(fn HandleBulletCount
  (bind _returnvalue (Ability|GetCurrentSourceWeapon))
  (bind _returnvalue_2 (FortWeapon|Ammo|HasInfiniteMagazineAmmo _returnvalue))
  (bind _returnvalue_3 (not _returnvalue_2))
  (bind _returnvalue_4 (FortWeapon|GetBulletsPerClip _returnvalue))
  ; If not using the infinite ammo config
  (if _returnvalue_3
    (Variables|Default|SetBulletCount (- (FortWeapon|GetBulletsPerClip (Ability|GetCurrentSourceWeapon)) (FortWeapon|GetLocalMagazineAmmoCount (Ability|GetCurrentSourceWeapon))))
    (elif (and _returnvalue_3 (== (+ (FortWeapon|GetLocalMagazineAmmoCount _returnvalue) 1) _returnvalue_4))
      (Variables|Default|SetBulletCount 0)
      ; Check if reached bullets per clip
      (if (>= (Variables|Default|GetBulletCount) _returnvalue_4) :!label _label
        (Variables|Default|SetBulletCount 1)
        (else
          (bind _result (Math|Integer|IncrementInt (Variables|Default|GetBulletCount))
            (:%20%20
              (Variables|Default|SetBulletCount _result)))))
      (else
        (goto _label))))
  (return))

(fn IsDamgeFromSourceWeapon
  (params (GameplayEventData :type Gameplay%20Event%20Data%20Structure%20%28by%20ref%29))
  (bind (_eventtag _instigator _target _optionalobject _optionalobject2 _contexthandle) (Utilities|Struct|BreakGameplayEventData GameplayEventData))
  (bind _asfort_pawn (CallFunction|GetAvatarasFortPawn))
  (Utilities|IsValid _asfort_pawn
    (:Is%20Valid
      (return (Utilities|Equal(Object) (Variables|/Script/FortniteGame.FortPawn|GetCurrentWeapon _asfort_pawn) (Ability|EffectContext|GetSourceObject _contexthandle))))
    (:Is%20Not%20Valid
      (return false))))

(fn IsPawnAndNotFriendly
  (params (OwningController :type Actor%20Object%20Reference) (HitActor :type Actor%20Object%20Reference))
  (return (and (Utilities|Enum|Equal(Enum) (Team|GetTeamAffiliation OwningController HitActor) "Hostile") (Utilities|IsA HitActor "/Script/FortniteGame.FortPawn"))))

(fn PlayGameplayCueOnSelf
  ; Check if effect has been triggered
  (if (not (GameplayTags|HasMatchingGameplayTag (Ability|GetAbilitySystemComponent (CallFunction|GetAvatarasFortPawn)) (Variables|Default|GetTag%20to%20Check)))
    ; Grant GE with triggered tag
    (GameplayEffects|ApplyGameplayEffectToTarget :self (Ability|GetAbilitySystemComponent (Ability|GetAvatarActorfromActorInfo)) :GameplayEffectClass (Variables|Default|GetGE_ApplyGPCue) :Target (Ability|GetAbilitySystemComponent (Ability|GetAvatarActorfromActorInfo)) :Level 1.0)
    ; Play Burst Cue
    (GameplayCue|ExecuteGameplayCueOnActor(Burst) (CallFunction|GetAvatarasFortPawn) (Variables|Default|GetGC_SuccessfulPullSelfCue) (Ability|GameplayCue|MakeGameplayCueParameters :NormalizedMagnitude 1.0 :RawMagnitude 1.0 :Location "0,0,0" :Normal "0,0,0" :Instigator (CallFunction|GetAvatarasFortPawn) :EffectCauser (CallFunction|GetAvatarasFortPawn) :SourceObject (Weapon|GetCurrentWeapon (CallFunction|GetAvatarasFortPawn)) :GameplayEffectLevel 1 :AbilityLevel 1 :bReplicateLocationWhenUsingMinimalRepProxy false)))
  (return))
