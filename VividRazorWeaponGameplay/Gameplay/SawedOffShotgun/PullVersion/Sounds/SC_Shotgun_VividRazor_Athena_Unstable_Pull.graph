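# Sound cue for the unstable pull sound of the VividRazor sawed-off shotgun's pull version, classed under gadget items.
# A mixer blends two layers, Close and Med, each with its own gadget attenuation asset.
# Each layer applies a short fixed delay and modulation to a random pick of four waves.
# The cue raises overall pitch and carries concurrency override values for a small voice cap that prevents new plays.
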SoundCue /VividRazorWeaponGameplay/Gameplay/SawedOffShotgun/PullVersion/Sounds/SC_Shotgun_VividRazor_Athena_Unstable_Pull
root=Mixer_0
nodes [
 {id=Attenuation_0, children [Delay_0], properties {attenuationSettings=/Game/Sounds/Attenuation/Fort_Gadget_Sml_Att}}
 {id=Attenuation_1, children [Delay_1], properties {attenuationSettings=/Game/Sounds/Attenuation/Fort_GadgetPlace_Att}}
 {id=Delay_0, children [Modulator_1], properties {delayMax=0.05, delayMin=0.05}}
 {id=Delay_1, children [Modulator_0], properties {delayMax=0.05, delayMin=0.05}}
 {id=Mixer_0, children [Attenuation_0, Attenuation_1], properties {inputVolume [1,1]}}
 {id=Modulator_0, children [Random_4], properties {pitchMax=1.1}}
 {id=Modulator_1, children [Random_3], properties {pitchMax=1.1}}
 {id=Random_3, children [WavePlayer_8, WavePlayer_9, WavePlayer_10, WavePlayer_11], properties {weights [1,1,1,1]}}
 {id=Random_4, children [WavePlayer_12, WavePlayer_13, WavePlayer_14, WavePlayer_15], properties {weights [1,1,1,1]}}
 {
  id=WavePlayer_10
  properties {soundWaveAssetPtr=&/Shotgun_VividRazor_Athena_Unstable_Pull_Close_03}
 }
 {
  id=WavePlayer_11
  properties {soundWaveAssetPtr=&/Shotgun_VividRazor_Athena_Unstable_Pull_Close_04}
 }
 {
  id=WavePlayer_12
  properties {soundWaveAssetPtr=&/Shotgun_VividRazor_Athena_Unstable_Pull_Med_01}
 }
 {
  id=WavePlayer_13
  properties {soundWaveAssetPtr=&/Shotgun_VividRazor_Athena_Unstable_Pull_Med_02}
 }
 {
  id=WavePlayer_14
  properties {soundWaveAssetPtr=&/Shotgun_VividRazor_Athena_Unstable_Pull_Med_03}
 }
 {
  id=WavePlayer_15
  properties {soundWaveAssetPtr=&/Shotgun_VividRazor_Athena_Unstable_Pull_Med_04}
 }
 {
  id=WavePlayer_8
  properties {soundWaveAssetPtr=&/Shotgun_VividRazor_Athena_Unstable_Pull_Close_01}
 }
 {
  id=WavePlayer_9
  properties {soundWaveAssetPtr=&/Shotgun_VividRazor_Athena_Unstable_Pull_Close_02}
 }
]
properties {
 maxDistance=2000
 pitchMultiplier=1.3
 soundClassObject=&Fort_Gadgets_Items
 volumeMultiplier=1
 concurrencyOverrides {maxCount=3, resolutionRule=PreventNew}
}
